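# Sound cue for the foley heard when a player is hoisted up in the carry-and-throw (Yeet) mechanic.
# A Random node picks one of three equally weighted hoist recordings, then a Modulator applies slight pitch variation.
# It plays in the Athena player foley sound class and borrows the third-person DBNO shuffle attenuation.
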SoundCue /Game/Athena/Sounds/Character/Carry_Yeet/Foley_Carry_Hoist_Player_Cue
root=Modulator_0
nodes [
 {id=Modulator_0, children [Random_1], properties {pitchMax=1.02, pitchMin=0.98}}
 {id=Random_1, children [WavePlayer_8, WavePlayer_9, WavePlayer_10], properties {bShouldExcludeFromBranchCulling=true, weights [1,1,1]}}
 {id=WavePlayer_10, properties {soundWaveAssetPtr=&/Foly_Carry_Hoist_Player_03}}
 {id=WavePlayer_8, properties {soundWaveAssetPtr=&/Foly_Carry_Hoist_Player_01}}
 {id=WavePlayer_9, properties {soundWaveAssetPtr=&/Foly_Carry_Hoist_Player_02}}
]
properties {
 attenuationSettings=/Game/Sounds/Attenuation/Fort_Player_DBNO_Shuffle_3P_Athena_Att
 maxDistance=1500
 soundClassObject=&Fort_Player_Foley_Athena
 volumeMultiplier=1
}
