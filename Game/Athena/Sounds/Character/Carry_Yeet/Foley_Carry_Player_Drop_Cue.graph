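# Sound cue for the foley heard when a carried player is dropped in Athena.
# It picks one of two drop recordings at equal weight, then applies a slight random pitch variation through a modulator.
# It plays at reduced volume with weapon-drop attenuation and the third-person enemy player foley sound class.
# The random node is excluded from branch culling.
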SoundCue /Game/Athena/Sounds/Character/Carry_Yeet/Foley_Carry_Player_Drop_Cue
root=Modulator_0
nodes [
 {id=Modulator_0, children [Random_0], properties {pitchMax=1.02, pitchMin=0.98}}
 {id=Random_0, children [WavePlayer_2,WavePlayer_3], properties {bShouldExcludeFromBranchCulling=true, weights [1,1]}}
 {id=WavePlayer_2, properties {soundWaveAssetPtr=&/Foly_Carry_Drop_Player_01}}
 {id=WavePlayer_3, properties {soundWaveAssetPtr=&/Foly_Carry_Drop_Player_02}}
]
properties {
 attenuationSettings=&Fort_Weapon_Drop_Att
 maxDistance=3000
 soundClassObject=&Fort_Player_Foley_3P_Enemy_Athena
 volumeMultiplier=0.5
}
