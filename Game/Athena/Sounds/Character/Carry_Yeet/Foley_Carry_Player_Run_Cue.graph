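# Sound cue for the running foley of carrying a player, routed to the third-person enemy foley sound class at reduced volume.
# A quality-level node sends its first two slots to a random pick of eight run waves and its last two to a pick of the first four.
# A root modulator varies pitch and volume on each play, and distance falloff reuses the third-person DBNO shuffle attenuation.
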SoundCue /Game/Athena/Sounds/Character/Carry_Yeet/Foley_Carry_Player_Run_Cue
root=Modulator_0
nodes [
 {id=Modulator_0, children [QualityLevel_0], properties {pitchMax=1.1, pitchMin=0.9, volumeMax=1, volumeMin=0.6}}
 {id=QualityLevel_0, children [Random_0,Random_0,Random_1,Random_1]}
 {
  id=Random_0
  children [WavePlayer_0,WavePlayer_1,WavePlayer_2,WavePlayer_3,WavePlayer_4,WavePlayer_5,WavePlayer_6,WavePlayer_7]
  properties {weights [1,1,1,1,1,1,1,1]}
 }
 {
  id=Random_1
  children [WavePlayer_8, WavePlayer_9, WavePlayer_10, WavePlayer_11]
  properties {bShouldExcludeFromBranchCulling=true, weights [1,1,1,1]}
 }
 {id=WavePlayer_0, properties {soundWaveAssetPtr=&/Foly_Carry_Player_Run_01}}
 {id=WavePlayer_1, properties {soundWaveAssetPtr=&/Foly_Carry_Player_Run_02}}
 {id=WavePlayer_10, properties {soundWaveAssetPtr=&/Foly_Carry_Player_Run_03}}
 {id=WavePlayer_11, properties {soundWaveAssetPtr=&/Foly_Carry_Player_Run_04}}
 {id=WavePlayer_2, properties {soundWaveAssetPtr=&Foly_Carry_Player_Run_03}}
 {id=WavePlayer_3, properties {soundWaveAssetPtr=&Foly_Carry_Player_Run_04}}
 {id=WavePlayer_4, properties {soundWaveAssetPtr=&/Foly_Carry_Player_Run_05}}
 {id=WavePlayer_5, properties {soundWaveAssetPtr=&/Foly_Carry_Player_Run_06}}
 {id=WavePlayer_6, properties {soundWaveAssetPtr=&/Foly_Carry_Player_Run_07}}
 {id=WavePlayer_7, properties {soundWaveAssetPtr=&/Foly_Carry_Player_Run_08}}
 {id=WavePlayer_8, properties {soundWaveAssetPtr=&Foly_Carry_Player_Run_01}}
 {id=WavePlayer_9, properties {soundWaveAssetPtr=&Foly_Carry_Player_Run_02}}
]
properties {
 attenuationSettings=/Game/Sounds/Attenuation/Fort_Player_DBNO_Shuffle_3P_Athena_Att
 maxDistance=1500
 soundClassObject=&Fort_Player_Foley_3P_Enemy_Athena
 volumeMultiplier=0.15
}
