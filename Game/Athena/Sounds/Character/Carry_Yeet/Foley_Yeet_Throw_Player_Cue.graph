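# Sound cue for the foley heard when a carried player is thrown (yeeted) in Athena.
# A root Random node picks one of three equally weighted throw wave variations and is excluded from branch culling.
# It plays at reduced volume through the third-person enemy player foley sound class, borrowing the third-person DBNO shuffle attenuation for spatial falloff.
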SoundCue /Game/Athena/Sounds/Character/Carry_Yeet/Foley_Yeet_Throw_Player_Cue
root=Random_1
nodes [
 {id=Random_1, children [WavePlayer_8, WavePlayer_9, WavePlayer_10], properties {bShouldExcludeFromBranchCulling=true, weights [1,1,1]}}
 {id=WavePlayer_10, properties {soundWaveAssetPtr=&/Foly_Yeet_Throwe_Player_03}}
 {id=WavePlayer_8, properties {soundWaveAssetPtr=&/Foly_Yeet_Throwe_Player_01}}
 {id=WavePlayer_9, properties {soundWaveAssetPtr=&/Foly_Yeet_Throwe_Player_02}}
]
properties {
 attenuationSettings=/Game/Sounds/Attenuation/Fort_Player_DBNO_Shuffle_3P_Athena_Att
 maxDistance=1500
 soundClassObject=&Fort_Player_Foley_3P_Enemy_Athena
 volumeMultiplier=0.7
}
